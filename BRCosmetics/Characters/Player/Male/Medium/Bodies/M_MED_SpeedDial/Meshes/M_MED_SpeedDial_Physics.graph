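# Physics asset for the SpeedDial medium male player body, covering the dynamic bones of the jacket's lower sections.
# Capsules on the pelvis and thighs accompany four two-segment jacket chains (two center, one left, one right).
# Each chain uses tight angular limits and hangs from a base body whose pelvis joint has free rotation.
# The right chain is also constrained to the right thigh; three other small base bodies have no constraints.
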PhysicsAsset /BRCosmetics/Characters/Player/Male/Medium/Bodies/M_MED_SpeedDial/Meshes/M_MED_SpeedDial_Physics
bodies {
 dyn_jacketMid_base_01_r [{shapeType=Capsule, length=0.505, radius=0.505, shapeName=dyn_jacketMid_base_01_r_capsule, rotation=<rot 0 -90 0>}]
 dyn_jacket_01_c [{
   shapeType=Capsule
   length=15.5
   radius=4.1172
   shapeName=dyn_jacket_01_c_capsule
   center=<3.2167 -3.2116 0.0703>
   rotation=<rot 0 5.2355 2.8125>
  }]
 dyn_jacket_01_end_c [{
   shapeType=Capsule
   length=15.5
   radius=4.5063
   shapeName=dyn_jacket_01_end_c_capsule
   center=<3.5734 -3.2409 0.0813>
   rotation=<rot 0 4.1048 2.8125>
  }]
 dyn_jacket_02_c [{
   shapeType=Capsule
   length=16.6882
   radius=4.1297
   shapeName=dyn_jacket_02_c_capsule
   center=<3.0135 -1.7263 -0.1064>
   rotation=<rot 0 6.3506 -5.625>
  }]
 dyn_jacket_02_end_c [{
   shapeType=Capsule
   length=17.7734
   radius=4.4284
   shapeName=dyn_jacket_02_end_c_capsule
   center=<2.8138 -1.0695 -0.1364>
   rotation=<rot 0 5.1243 -5.625>
  }]
 dyn_jacket_03_end_l [{shapeType=Capsule, length=9.4626, radius=5.0768, shapeName=dyn_jacket_03_end_l_capsule, center=<2.6998 -4.9589 0>}]
 dyn_jacket_03_end_r [{
   shapeType=Capsule
   length=7.6267
   radius=4.432
   shapeName=dyn_jacket_03_end_r_capsule
   center=<2.5459 -3.1677 0>
   rotation=<rot -0.0002 5.6712 0>
  }]
 dyn_jacket_03_l [{shapeType=Capsule, length=8.8863, radius=5.0768, shapeName=dyn_jacket_03_l_capsule, center=<2.0989 -3.826 0>}]
 dyn_jacket_03_r [{shapeType=Capsule, length=6.7347, radius=3.9355, shapeName=dyn_jacket_03_r_capsule, center=<3.8088 -3.2788 0>}]
 dyn_jacket_base_01_c [{shapeType=Capsule, length=0.505, radius=0.505, shapeName=dyn_jacket_base_01_c_capsule, rotation=<rot 0 -90 0>}]
 dyn_jacket_base_01_r [{shapeType=Capsule, length=0.505, radius=0.505, shapeName=dyn_jacket_base_01_r_capsule, rotation=<rot 0 -90 0>}]
 dyn_jacket_base_02_c [{shapeType=Capsule, length=0.505, radius=0.505, shapeName=dyn_jacket_base_02_c_capsule, rotation=<rot 0 -90 0>}]
 dyn_jacket_base_02_r [{shapeType=Capsule, length=0.505, radius=0.505, shapeName=dyn_jacket_base_02_r_capsule, rotation=<rot 0 -90 0>}]
 dyn_jacket_base_03_l [{shapeType=Capsule, length=0.505, radius=0.505, shapeName=dyn_jacket_base_03_l_capsule, rotation=<rot 0 -90 0>}]
 dyn_jacket_base_03_r [{shapeType=Capsule, length=0.505, radius=0.505, shapeName=dyn_jacket_base_03_r_capsule, rotation=<rot 0 -90 0>}]
 pelvis [
  {shapeType=Capsule, length=13.879, radius=9.137, shapeName=pelvis_capsule, center=<-4.0651 -1.9614 0>}
  {shapeType=Capsule, length=4.9666, radius=10.9429, shapeName=pelvis_capsule, center=<-4.7209 0 0>}
 ]
 thigh_l [{
   shapeType=Capsule
   length=25.9198
   radius=8.2206
   shapeName=thigh_l_capsule
   center=<-18.9406 -1.7929 0.8619>
   rotation=<rot -13.8921 91.0342 -92.4248>
  }]
 thigh_r [{
   shapeType=Capsule
   length=25.4256
   radius=8.7148
   shapeName=thigh_l_capsule
   center=<18.9059 1.5103 -0.7101>
   rotation=<rot -4.9315 94.2435 86.7705>
  }]
}
constraints [
 {bone1Name=dyn_jacket_01_c, bone2Name=dyn_jacket_base_01_c, swing1LimitDegrees=25, swing2LimitDegrees=5, twistLimitDegrees=5}
 {bone1Name=dyn_jacket_01_end_c, bone2Name=dyn_jacket_01_c, swing1LimitDegrees=30, swing2LimitDegrees=10, twistLimitDegrees=5}
 {bone1Name=dyn_jacket_02_c, bone2Name=dyn_jacket_base_02_c, swing1LimitDegrees=25, swing2LimitDegrees=5, twistLimitDegrees=5}
 {bone1Name=dyn_jacket_02_end_c, bone2Name=dyn_jacket_02_c, swing1LimitDegrees=30, swing2LimitDegrees=10, twistLimitDegrees=5}
 {bone1Name=dyn_jacket_03_r, bone2Name=dyn_jacket_base_03_r, swing1LimitDegrees=20, swing2LimitDegrees=19, twistLimitDegrees=5}
 {
  bone1Name=dyn_jacket_base_01_c
  bone2Name=pelvis
  swing1LimitDegrees=45
  swing1Motion=Free
  swing2LimitDegrees=45
  swing2Motion=Free
  twistLimitDegrees=45
  twistMotion=Free
 }
 {
  bone1Name=dyn_jacket_base_02_c
  bone2Name=pelvis
  swing1LimitDegrees=45
  swing1Motion=Free
  swing2LimitDegrees=45
  swing2Motion=Free
  twistLimitDegrees=45
  twistMotion=Free
 }
 {
  bone1Name=dyn_jacket_base_03_r
  bone2Name=pelvis
  swing1LimitDegrees=45
  swing1Motion=Free
  swing2LimitDegrees=45
  swing2Motion=Free
  twistLimitDegrees=45
  twistMotion=Free
 }
 {bone1Name=dyn_jacket_03_end_r, bone2Name=dyn_jacket_03_r, swing1LimitDegrees=25, swing2LimitDegrees=15, twistLimitDegrees=5}
 {
  bone1Name=thigh_r
  bone2Name=pelvis
  swing1LimitDegrees=45
  swing1Motion=Free
  swing2LimitDegrees=45
  swing2Motion=Free
  twistLimitDegrees=45
  twistMotion=Free
 }
 {
  bone1Name=dyn_jacket_base_03_l
  bone2Name=pelvis
  swing1LimitDegrees=45
  swing1Motion=Free
  swing2LimitDegrees=45
  swing2Motion=Free
  twistLimitDegrees=45
  twistMotion=Free
 }
 {bone1Name=dyn_jacket_03_l, bone2Name=dyn_jacket_base_03_l, swing1LimitDegrees=25, swing2LimitDegrees=15, twistLimitDegrees=5}
 {bone1Name=dyn_jacket_03_end_l, bone2Name=dyn_jacket_03_l, swing1LimitDegrees=25, swing2LimitDegrees=10, twistLimitDegrees=5}
 {bone1Name=dyn_jacket_03_r, bone2Name=thigh_r, swing1LimitDegrees=30, swing2LimitDegrees=30, twistLimitDegrees=30}
]
