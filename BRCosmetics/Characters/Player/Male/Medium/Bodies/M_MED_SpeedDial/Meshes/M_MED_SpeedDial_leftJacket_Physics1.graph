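# Physics asset for the left jacket of the male medium SpeedDial player body, simulating its dynamic jacket bones.
# Three base-mid-end capsule chains (jacket_01, jacketMid, jacket_02) hang from the pelvis via free-rotation base joints, with angle-limited joints below.
# Cross constraints tie neighboring chains at mid and end bones so the panel moves together.
# Pelvis and thigh shapes form the body volumes; a right-side base body has no constraint.
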PhysicsAsset /BRCosmetics/Characters/Player/Male/Medium/Bodies/M_MED_SpeedDial/Meshes/M_MED_SpeedDial_leftJacket_Physics1
bodies {
 dyn_jacketMid_01_end_l [{
   shapeType=Capsule
   length=16.4115
   radius=7
   shapeName=dyn_jacketMid_01_end_l_capsule
   center=<0.5305 -7.0372 0>
   rotation=<rot 90 -168.5237 -173.041>
  }]
 dyn_jacketMid_01_l [{shapeType=Capsule, length=8.8947, radius=3.3692, shapeName=dyn_jacketMid_01_l_capsule, center=<3.2716 -4.6855 0>}]
 dyn_jacketMid_base_01_l [{shapeType=Capsule, length=0.505, radius=0.505, shapeName=dyn_jacketMid_base_01_l_capsule, rotation=<rot 0 -90 0>}]
 dyn_jacket_01_end_l [{
   shapeType=Capsule
   length=13.3504
   radius=7
   shapeName=dyn_jacket_01_end_l_capsule
   center=<0.5305 -7.0372 0>
   rotation=<rot 90 -174.1035 -178.6208>
  }]
 dyn_jacket_01_l [{shapeType=Capsule, length=8.8947, radius=3.3692, shapeName=dyn_jacket_01_l_capsule, center=<3.2716 -4.6855 0>}]
 dyn_jacket_02_end_l [{
   shapeType=Capsule
   length=9.9271
   radius=7
   shapeName=dyn_jacket_02_end_l_capsule
   center=<0.5305 -7.0372 0>
   rotation=<rot 90 -171.7561 -176.2736>
  }]
 dyn_jacket_02_l [{shapeType=Capsule, length=8.8947, radius=3.3692, shapeName=dyn_jacket_02_l_capsule, center=<3.2716 -4.6855 0>}]
 dyn_jacket_base_01_l [{shapeType=Capsule, length=0.505, radius=0.505, shapeName=dyn_jacket_base_01_l_capsule, rotation=<rot 0 -90 0>}]
 dyn_jacket_base_01_r [{shapeType=Capsule, length=0.505, radius=0.505, shapeName=dyn_jacket_base_01_r_capsule, rotation=<rot 0 -90 0>}]
 dyn_jacket_base_02_l [{shapeType=Capsule, length=0.505, radius=0.505, shapeName=dyn_jacket_base_02_l_capsule, rotation=<rot 0 -90 0>}]
 pelvis [
  {shapeType=Sphere, radius=9.0496, shapeName=pelvis_sphere, center=<-10.7155 -0.1442 0>}
  {shapeType=Capsule, length=11.0836, radius=9.2133, shapeName=pelvis_capsule, center=<0 -0.369 0>}
 ]
 thigh_l [
  {
   shapeType=Capsule
   length=39.4529
   radius=9.6409
   shapeName=thigh_l_capsule
   center=<-18.9625 -0.5797 0.5618>
   rotation=<rot -13.8921 91.0342 -92.4248>
  }
  {shapeType=Capsule, length=9.0025, radius=8.4641, shapeName=thigh_l_capsule, center=<-6.9025 -1.2012 6.6533>, rotation=<rot 90 180 -177.1344>}
 ]
 thigh_r [{
   shapeType=Capsule
   length=41.6158
   radius=7.4781
   shapeName=thigh_l_capsule
   center=<18.9625 0.5797 -0.5618>
   rotation=<rot -13.8921 91.0342 87.5752>
  }]
}
constraints [
 {
  bone1Name=thigh_l
  bone2Name=pelvis
  swing1LimitDegrees=45
  swing1Motion=Free
  swing2LimitDegrees=45
  swing2Motion=Free
  twistLimitDegrees=45
  twistMotion=Free
 }
 {bone1Name=dyn_jacket_02_l, bone2Name=dyn_jacket_base_02_l, swing1LimitDegrees=60, swing2LimitDegrees=20, twistLimitDegrees=5}
 {bone1Name=dyn_jacket_01_end_l, bone2Name=dyn_jacket_01_l, swing1LimitDegrees=45, swing2LimitDegrees=5, twistLimitDegrees=5}
 {
  bone1Name=dyn_jacket_base_01_l
  bone2Name=pelvis
  swing1LimitDegrees=45
  swing1Motion=Free
  swing2LimitDegrees=45
  swing2Motion=Free
  twistLimitDegrees=45
  twistMotion=Free
 }
 {
  bone1Name=dyn_jacket_base_02_l
  bone2Name=pelvis
  swing1LimitDegrees=45
  swing1Motion=Free
  swing2LimitDegrees=45
  swing2Motion=Free
  twistLimitDegrees=45
  twistMotion=Free
 }
 {bone1Name=dyn_jacket_01_l, bone2Name=dyn_jacket_base_01_l, swing1LimitDegrees=60, swing2LimitDegrees=5, twistLimitDegrees=5}
 {bone1Name=dyn_jacket_02_end_l, bone2Name=dyn_jacket_02_l, swing1LimitDegrees=45, swing2LimitDegrees=20, twistLimitDegrees=5}
 {
  bone1Name=dyn_jacketMid_base_01_l
  bone2Name=pelvis
  swing1LimitDegrees=45
  swing1Motion=Free
  swing2LimitDegrees=45
  swing2Motion=Free
  twistLimitDegrees=45
  twistMotion=Free
 }
 {bone1Name=dyn_jacketMid_01_l, bone2Name=dyn_jacketMid_base_01_l, swing1LimitDegrees=60, swing2LimitDegrees=20, twistLimitDegrees=5}
 {bone1Name=dyn_jacketMid_01_end_l, bone2Name=dyn_jacketMid_01_l, swing1LimitDegrees=45, swing2LimitDegrees=20, twistLimitDegrees=5}
 {bone1Name=dyn_jacketMid_01_end_l, bone2Name=dyn_jacket_01_end_l, swing1LimitDegrees=10, swing2LimitDegrees=5, twistLimitDegrees=5}
 {
  bone1Name=thigh_r
  bone2Name=pelvis
  swing1LimitDegrees=45
  swing1Motion=Free
  swing2LimitDegrees=45
  swing2Motion=Free
  twistLimitDegrees=45
  twistMotion=Free
 }
 {
  bone1Name=dyn_jacketMid_01_l
  bone2Name=dyn_jacket_01_l
  swing1LimitDegrees=45
  swing1Motion=Free
  swing2LimitDegrees=10
  twistLimitDegrees=10
 }
 {
  bone1Name=dyn_jacket_02_l
  bone2Name=dyn_jacketMid_01_l
  swing1LimitDegrees=45
  swing1Motion=Free
  swing2LimitDegrees=10
  twistLimitDegrees=10
 }
 {bone1Name=dyn_jacket_02_end_l, bone2Name=dyn_jacketMid_01_end_l, swing1LimitDegrees=10, swing2LimitDegrees=10, twistLimitDegrees=10}
]
