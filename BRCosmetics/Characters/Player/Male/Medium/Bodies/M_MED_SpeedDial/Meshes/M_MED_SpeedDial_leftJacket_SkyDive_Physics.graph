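# Physics asset for the SpeedDial medium male body's left jacket, named for the skydive state.
# Four three-bone chains (three jacket, one jacketMid) each run from a tiny base capsule through two larger capsules, joined by narrow swing and twist limits.
# Chain bases and the left thigh connect to the pelvis with free angular motion; pelvis and thigh capsules supply body shapes.
# A right-side jacket base body is also present, without any constraint.
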PhysicsAsset /BRCosmetics/Characters/Player/Male/Medium/Bodies/M_MED_SpeedDial/Meshes/M_MED_SpeedDial_leftJacket_SkyDive_Physics
bodies {
 dyn_jacketMid_01_end_l [{
   shapeType=Capsule
   length=15
   radius=5
   shapeName=dyn_jacketMid_01_end_l_capsule
   center=<2.9814 -2.6403 -0.0047>
   rotation=<rot 0.0001 -0.0001 0>
  }]
 dyn_jacketMid_01_l [{
   shapeType=Capsule
   length=15
   radius=5
   shapeName=dyn_jacketMid_01_l_capsule
   center=<0.8221 -3.7955 -0.1556>
   rotation=<rot 0.0001 -0.0001 -11.25>
  }]
 dyn_jacketMid_base_01_l [{shapeType=Capsule, length=0.505, radius=0.505, shapeName=dyn_jacketMid_base_01_l_capsule, rotation=<rot 0 -90 0>}]
 dyn_jacket_01_end_l [{shapeType=Capsule, length=15, radius=5, shapeName=dyn_jacket_01_end_l_capsule, center=<2.614 -3.0132 0>}]
 dyn_jacket_01_l [{shapeType=Capsule, length=15, radius=5, shapeName=dyn_jacket_01_l_capsule, center=<0.8221 -3.0132 0>}]
 dyn_jacket_02_end_l [{shapeType=Capsule, length=15, radius=5, shapeName=dyn_jacket_02_end_l_capsule, center=<0.8221 -2.9235 0>}]
 dyn_jacket_02_l [{shapeType=Capsule, length=15, radius=5, shapeName=dyn_jacket_02_l_capsule, center=<0.8221 -2.9235 0>}]
 dyn_jacket_03_end_l [{shapeType=Capsule, length=15, radius=5, shapeName=dyn_jacket_03_end_l_capsule, center=<0.8221 -3.3534 0>}]
 dyn_jacket_03_l [{shapeType=Capsule, length=15, radius=5, shapeName=dyn_jacket_03_l_capsule, center=<0.8221 -3.3534 0>}]
 dyn_jacket_base_01_l [{shapeType=Capsule, length=0.505, radius=0.505, shapeName=dyn_jacket_base_01_l_capsule, rotation=<rot 0 -90 0>}]
 dyn_jacket_base_01_r [{shapeType=Capsule, length=0.505, radius=0.505, shapeName=dyn_jacket_base_01_r_capsule, rotation=<rot 0 -90 0>}]
 dyn_jacket_base_02_l [{shapeType=Capsule, length=0.505, radius=0.505, shapeName=dyn_jacket_base_02_l_capsule, rotation=<rot 0 -90 0>}]
 dyn_jacket_base_03_l [{shapeType=Capsule, length=0.505, radius=0.505, shapeName=dyn_jacket_base_03_l_capsule, rotation=<rot 0 -90 0>}]
 pelvis [{shapeType=Capsule, length=7.1874, radius=9.137, shapeName=pelvis_capsule, center=<0 -0.369 0>}]
 thigh_l [{
   shapeType=Capsule
   length=25.9198
   radius=8.2206
   shapeName=thigh_l_capsule
   center=<-18.9625 -0.5797 0.5618>
   rotation=<rot -13.8921 91.0342 -92.4248>
  }]
}
constraints [
 {
  bone1Name=thigh_l
  bone2Name=pelvis
  swing1LimitDegrees=45
  swing1Motion=Free
  swing2LimitDegrees=45
  swing2Motion=Free
  twistLimitDegrees=45
  twistMotion=Free
 }
 {bone1Name=dyn_jacket_02_l, bone2Name=dyn_jacket_base_02_l, swing1LimitDegrees=12, swing2LimitDegrees=5, twistLimitDegrees=5}
 {bone1Name=dyn_jacket_01_end_l, bone2Name=dyn_jacket_01_l, swing1LimitDegrees=12, swing2LimitDegrees=5, twistLimitDegrees=5}
 {
  bone1Name=dyn_jacket_base_01_l
  bone2Name=pelvis
  swing1LimitDegrees=45
  swing1Motion=Free
  swing2LimitDegrees=45
  swing2Motion=Free
  twistLimitDegrees=45
  twistMotion=Free
 }
 {
  bone1Name=dyn_jacket_base_02_l
  bone2Name=pelvis
  swing1LimitDegrees=45
  swing1Motion=Free
  swing2LimitDegrees=45
  swing2Motion=Free
  twistLimitDegrees=45
  twistMotion=Free
 }
 {bone1Name=dyn_jacket_01_l, bone2Name=dyn_jacket_base_01_l, swing1LimitDegrees=12, swing2LimitDegrees=5, twistLimitDegrees=5}
 {bone1Name=dyn_jacket_02_end_l, bone2Name=dyn_jacket_02_l, swing1LimitDegrees=12, swing2LimitDegrees=5, twistLimitDegrees=3}
 {
  bone1Name=dyn_jacketMid_base_01_l
  bone2Name=pelvis
  swing1LimitDegrees=45
  swing1Motion=Free
  swing2LimitDegrees=45
  swing2Motion=Free
  twistLimitDegrees=45
  twistMotion=Free
 }
 {bone1Name=dyn_jacketMid_01_l, bone2Name=dyn_jacketMid_base_01_l, swing1LimitDegrees=25, swing2LimitDegrees=5, twistLimitDegrees=5}
 {bone1Name=dyn_jacketMid_01_end_l, bone2Name=dyn_jacketMid_01_l, swing1LimitDegrees=12, swing2LimitDegrees=5, twistLimitDegrees=5}
 {
  bone1Name=dyn_jacket_base_03_l
  bone2Name=pelvis
  swing1LimitDegrees=45
  swing1Motion=Free
  swing2LimitDegrees=45
  swing2Motion=Free
  twistLimitDegrees=45
  twistMotion=Free
 }
 {bone1Name=dyn_jacket_03_l, bone2Name=dyn_jacket_base_03_l, swing1LimitDegrees=12, swing2LimitDegrees=5, twistLimitDegrees=5}
 {bone1Name=dyn_jacket_03_end_l, bone2Name=dyn_jacket_03_l, swing1LimitDegrees=12, swing2LimitDegrees=5, twistLimitDegrees=5}
]
