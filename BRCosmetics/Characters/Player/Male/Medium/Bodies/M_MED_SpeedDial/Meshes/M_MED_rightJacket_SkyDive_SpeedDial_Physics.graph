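# Physics asset for the right jacket flap of the SpeedDial male medium body's skydive mesh, apparently driving its secondary motion.
# Three bone chains (jacket 01, Mid, 02) hang from pelvis through tiny base capsules with free rotation, then narrowly limited links.
# Angular-locked cross-constraints tie both outer chain ends to the middle end, coupling the strips as one panel.
# Pelvis and thigh capsules provide body volumes; only the right thigh is constrained to pelvis.
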PhysicsAsset /BRCosmetics/Characters/Player/Male/Medium/Bodies/M_MED_SpeedDial/Meshes/M_MED_rightJacket_SkyDive_SpeedDial_Physics
bodies {
 dyn_jacketMid_01_end_r [{
   shapeType=Box
   extentX=8.5848
   extentY=1.6487
   extentZ=12.4594
   shapeName=dyn_jacketMid_01_end_r_box
   center=<2.3997 0.1674 0.0205>
   rotation=<rot 2.0516 -0.0001 0>
  }]
 dyn_jacketMid_01_r [{shapeType=Box, extentX=8.4683, extentY=0.9066, extentZ=11.3661, shapeName=dyn_jacketMid_01_r_box, center=<3.3283 -0.5647 0>}]
 dyn_jacketMid_base_01_r [{shapeType=Capsule, length=0.505, radius=0.505, shapeName=dyn_jacketMid_base_01_r_capsule, rotation=<rot 0 -90 0>}]
 dyn_jacket_01_end_r [{shapeType=Box, extentX=9.719, extentY=2.1725, extentZ=9.7618, shapeName=dyn_jacket_01_end_r_box, center=<2.1505 0.1683 0>}]
 dyn_jacket_01_r [{shapeType=Box, extentX=6.7681, extentY=1.452, extentZ=8.7423, shapeName=dyn_jacket_01_r_box, center=<3.4059 0 0>}]
 dyn_jacket_02_end_r [
  {
   shapeType=Capsule
   length=7.5786
   radius=4.3917
   shapeName=dyn_jacket_02_end_r_capsule
   center=<4.0141 -3.7397 0>
   rotation=<rot -0.0003 5.3056 0>
  }
  {
   shapeType=Capsule
   length=8.9714
   radius=5.0953
   shapeName=dyn_jacket_02_end_r_capsule
   center=<4.2017 -0.9829 9.8191>
   rotation=<rot 0.4448 6.2686 36.5006>
  }
 ]
 dyn_jacket_02_r [{
   shapeType=Capsule
   length=6.6164
   radius=4.1211
   shapeName=dyn_jacket_02_r_capsule
   center=<4.4681 -3.5033 0>
   rotation=<rot -0.0002 4.6877 0>
  }]
 dyn_jacket_base_01_r [{shapeType=Capsule, length=0.505, radius=0.505, shapeName=dyn_jacket_base_01_r_capsule, rotation=<rot 0 -90 0>}]
 dyn_jacket_base_02_r [{shapeType=Capsule, length=0.505, radius=0.505, shapeName=dyn_jacket_base_02_r_capsule, rotation=<rot 0 -90 0>}]
 pelvis [{shapeType=Capsule, length=7.1874, radius=9.137, shapeName=pelvis_capsule, center=<0 -0.369 0>}]
 thigh_l [{
   shapeType=Capsule
   length=25.9198
   radius=8.2206
   shapeName=thigh_l_capsule
   center=<-18.9481 -1.3731 0.7581>
   rotation=<rot -13.8921 91.0342 -92.4248>
  }]
 thigh_r [{
   shapeType=Capsule
   length=25.4256
   radius=8.7148
   shapeName=thigh_l_capsule
   center=<18.9059 1.5103 -0.7101>
   rotation=<rot -4.9315 94.2435 86.7705>
  }]
}
constraints [
 {
  bone1Name=dyn_jacket_base_01_r
  bone2Name=pelvis
  swing1LimitDegrees=45
  swing1Motion=Free
  swing2LimitDegrees=45
  swing2Motion=Free
  twistLimitDegrees=45
  twistMotion=Free
 }
 {
  bone1Name=dyn_jacket_base_02_r
  bone2Name=pelvis
  swing1LimitDegrees=45
  swing1Motion=Free
  swing2LimitDegrees=45
  swing2Motion=Free
  twistLimitDegrees=45
  twistMotion=Free
 }
 {bone1Name=dyn_jacketMid_01_r, bone2Name=dyn_jacketMid_base_01_r, swing1LimitDegrees=12, swing2LimitDegrees=5, twistLimitDegrees=5}
 {bone1Name=dyn_jacket_01_end_r, bone2Name=dyn_jacket_01_r, swing1LimitDegrees=12, swing2LimitDegrees=5, twistLimitDegrees=5}
 {bone1Name=dyn_jacket_01_r, bone2Name=dyn_jacket_base_01_r, swing1LimitDegrees=12, swing2LimitDegrees=5, twistLimitDegrees=5}
 {bone1Name=dyn_jacketMid_01_end_r, bone2Name=dyn_jacketMid_01_r, swing1LimitDegrees=12, swing2LimitDegrees=5, twistLimitDegrees=5}
 {
  bone1Name=dyn_jacketMid_base_01_r
  bone2Name=pelvis
  swing1LimitDegrees=45
  swing1Motion=Free
  swing2LimitDegrees=45
  swing2Motion=Free
  twistLimitDegrees=45
  twistMotion=Free
 }
 {bone1Name=dyn_jacket_02_end_r, bone2Name=dyn_jacket_02_r, swing1LimitDegrees=15, swing2LimitDegrees=2, twistLimitDegrees=5}
 {bone1Name=dyn_jacket_02_r, bone2Name=dyn_jacket_base_02_r, swing1LimitDegrees=15, swing2LimitDegrees=5, twistLimitDegrees=5}
 {
  bone1Name=thigh_r
  bone2Name=pelvis
  swing1LimitDegrees=45
  swing1Motion=Free
  swing2LimitDegrees=45
  swing2Motion=Free
  twistLimitDegrees=45
  twistMotion=Free
 }
 {
  bone1Name=dyn_jacket_01_end_r
  bone2Name=dyn_jacketMid_01_end_r
  swing1LimitDegrees=45
  swing1Motion=Locked
  swing2LimitDegrees=45
  swing2Motion=Locked
  twistLimitDegrees=45
  twistMotion=Locked
 }
 {
  bone1Name=dyn_jacket_02_end_r
  bone2Name=dyn_jacketMid_01_end_r
  swing1LimitDegrees=45
  swing1Motion=Locked
  swing2LimitDegrees=45
  swing2Motion=Locked
  twistLimitDegrees=45
  twistMotion=Locked
 }
]
