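# Character-part AnimBP for the SpeedDial male body that copies the parent mesh pose and adds jacket dynamics and fixes.
# A Control Rig feeds pistol and grind-rail bone offsets, then RigidBody sets switched between ground and skydive.
# A deform Control Rig with interpolated jacket correction follows, then stacked additive bone corrections, some for throwing or crouching.
# The event graph updates those alphas, the gravity override and calf-driven WPO cloth material parameters, adding wind during skydiving.
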AnimBlueprint /BRCosmetics/Characters/Player/Male/Medium/Bodies/M_MED_SpeedDial/Meshes/M_MED_SpeedDial_AnimBP
target_skeleton=&/M_MED_SpeedDial_Skeleton
graphs [
 AnimGraph {nodes [
   "Transform (Modify) Bone - Bone: dyn_jacket_base_02_r"!(ComponentPose, bAlphaBoolEnabled:bool=Get bIsSliding(), Translation="-0.474,-1.102,-0.081", Rotation="-14.150,-24.405,14.908", Scale="1,1,1") -> (Pose) {
    alphaInputType=Bool
    rotationMode=BMM_Additive
    translationMode=BMM_Additive
    alphaBoolBlend {blendOption=Linear}
    boneToModify {boneName=dyn_jacket_base_02_r}
    rotation=<rot -14.15 -24.405 14.908>
    translation=<-0.474 -1.102 -0.081>
   }
   "Transform (Modify) Bone - Bone: dyn_jacket_base_01_c"(Alpha=1, Translation="0,-1,0", Rotation="0,0,0", Scale="1,1,1") {
    rotationMode=BMM_Additive
    rotationSpace=BCS_BoneSpace
    translationMode=BMM_Additive
    translationSpace=BCS_BoneSpace
    alphaBoolBlend {blendOption=Linear}
    alphaScaleBiasClamp {bInterpResult=true}
    boneToModify {boneName=dyn_jacket_base_01_c}
    translation {y=-1}
   }
   n1 = Get Override World Gravity() -> (Override World Gravity:Vector)
   n2 = Get skydive Alpha() -> (skydive Alpha:double)
   n3 = Control Rig!(Source=Copy Pose From Mesh() {bUseAttachedParent=true}, Alpha, RigidBody enabled:bool=Get bRigidBodyEnabled()) -> (Pose) {
    alphaBoolBlend {blendOption=Linear}
    controlRigAssetReference {blueprintRigClass=&/M_MED_SpeedDial_CtrlRig.M_MED_SpeedDial_CtrlRig_C}
   }
   n4 = "Transform (Modify) Bone - Bone: osob_spine_4_joint"!(
    ComponentPose=Local To Component(LocalPose=@n3)
    bAlphaBoolEnabled:bool=Get is_pistol() -> (is_pistol:bool)
    Translation_X:double=Get lowerarm() -> (lowerarm:double)
    Translation_Y:double=0
    Translation_Z:double=0
    Rotation="0,0,0"
    Scale="1,1,1"
   ) -> (Pose) {
    alphaInputType=Bool
    rotationMode=BMM_Additive
    translationMode=BMM_Additive
    alphaBoolBlend {blendOption=Linear}
    boneToModify {boneName=osob_spine_4_joint}
    translation {x=14}
   }
   n5 = Get bIsOnGrindRail()
   n6 = "Transform (Modify) Bone - Bone: dyn_jacketMid_base_01_r"!(ComponentPose=@n4, bAlphaBoolEnabled:bool=@n5, Translation="0,0,2", Rotation="0,-40,45", Scale="1,1,1") -> (Pose) {
    alphaInputType=Bool
    rotationMode=BMM_Additive
    translationMode=BMM_Additive
    alphaBoolBlend {blendOption=Linear}
    boneToModify {boneName=dyn_jacketMid_base_01_r}
    rotation {roll=45, yaw=-40}
    translation {z=2}
   }
   n7 = "Transform (Modify) Bone - Bone: dyn_jacket_base_01_r"!(ComponentPose=@n6, bAlphaBoolEnabled:bool=@n5, Translation="0,0,4", Rotation="0,0,50", Scale="1,1,1") -> (Pose) {
    alphaInputType=Bool
    rotationMode=BMM_Additive
    translationMode=BMM_Additive
    alphaBoolBlend {blendOption=Linear}
    boneToModify {boneName=dyn_jacket_base_01_r}
    rotation {roll=50}
    translation {z=4}
   }
   n8 = "Transform (Modify) Bone - Bone: dyn_jacket_base_02_r"(ComponentPose=@n7, bAlphaBoolEnabled=@n5, Translation="0,0,4", Rotation="0,0,30", Scale="1,1,1") {
    alphaInputType=Bool
    rotationMode=BMM_Additive
    translationMode=BMM_Additive
    alphaBoolBlend {blendOption=Linear}
    boneToModify {boneName=dyn_jacket_base_02_r}
    rotation {roll=30}
    translation {z=4}
   }
   n9 = "Transform (Modify) Bone - Bone: dyn_jacket_base_02_l"!(ComponentPose=@n8, bAlphaBoolEnabled:bool=@n5, Translation="0,0,4", Rotation="-20,0,0", Scale="1,1,1") -> (Pose) {
    alphaInputType=Bool
    rotationMode=BMM_Additive
    translationMode=BMM_Additive
    alphaBoolBlend {blendOption=Linear}
    boneToModify {boneName=dyn_jacket_base_02_l}
    rotation {pitch=-20}
    translation {z=4}
   }
   n10 = Get Ground Alpha() -> (Ground Alpha:double)
   n11 = Get Override World Gravity()
   n12 = RigidBody!(ComponentPose=@n9, Alpha=@n10, OverrideWorldGravity:Vector=@n11, ExternalForce) -> (Pose) {
    bOverrideWorldGravity=true
    overridePhysicsAsset=&/M_MED_SpeedDial_Physics
    simulationSpace=BaseBoneSpace
    alphaBoolBlend {blendOption=Linear}
    baseBoneRef {boneName=pelvis}
    componentAppliedLinearAccClamp=<30000 30000 3000>
    componentLinearAccScale=<0.3 0.3 0.3>
    componentLinearVelScale=<0.2 0.2 0.5>
    simSpaceSettings {
     maxAngularAcceleration=100
     maxAngularVelocity=2
     maxLinearAcceleration=100
     maxLinearVelocity=20
     velocityScaleZ=0
     worldAlpha=0.75
     externalAngularVelocity=<1 1 1>
     externalLinearDragV=<1 1 1>
    }
   }
   n13 = RigidBody(ComponentPose=@n12, Alpha=@n10, OverrideWorldGravity=@n11) {
    bOverrideWorldGravity=true
    overridePhysicsAsset=&/M_MED_SpeedDial_leftJacket_Physics1
    simulationSpace=BaseBoneSpace
    alphaBoolBlend {blendOption=Linear}
    baseBoneRef {boneName=thigh_l}
    componentAppliedLinearAccClamp=<30000 30000 30000>
    componentLinearAccScale=<0.3 0.3 0.3>
    componentLinearVelScale=<0.2 0.2 0.3>
    simSpaceSettings {
     maxAngularAcceleration=100
     maxAngularVelocity=2
     maxLinearAcceleration=100
     maxLinearVelocity=160
     velocityScaleZ=0
     worldAlpha=0.75
     externalLinearDragV=<1 1 1>
    }
   }
   n14 = RigidBody(ComponentPose=@n13, Alpha=@n10, OverrideWorldGravity=@n11) {
    bOverrideWorldGravity=true
    overridePhysicsAsset=&/M_MED_rightJacket_SpeedDial_Physics
    simulationSpace=BaseBoneSpace
    alphaBoolBlend {blendOption=Linear}
    baseBoneRef {boneName=thigh_r}
    componentAppliedLinearAccClamp=<30000 30000 30000>
    componentLinearAccScale=<0.3 0.3 0.3>
    componentLinearVelScale=<0.2 0.2 0.5>
    simSpaceSettings {
     maxAngularAcceleration=100
     maxAngularVelocity=2
     maxLinearAcceleration=100
     maxLinearVelocity=160
     velocityScaleZ=0
     worldAlpha=0.75
     externalLinearDragV=<1 1 1>
    }
   }
   n15 = RigidBody(ComponentPose=@n14, Alpha=@n2, OverrideWorldGravity=@n1) {
    bClampLinearTranslationLimitToRefPose=true
    bOverrideWorldGravity=true
    overridePhysicsAsset=&/M_MED_rightJacket_SkyDive_SpeedDial_Physics
    simulationSpace=BaseBoneSpace
    alphaBoolBlend {blendOption=Linear}
    baseBoneRef {boneName=thigh_r}
    componentAppliedLinearAccClamp=<30000 30000 30000>
    componentLinearAccScale=<0.3 0.3 0.3>
    componentLinearVelScale=<0.2 0.2 0.5>
    simSpaceSettings {
     maxAngularAcceleration=100
     maxAngularVelocity=2
     maxLinearAcceleration=100
     maxLinearVelocity=160
     velocityScaleZ=0
     worldAlpha=0.75
     externalLinearDragV=<1 1 1>
    }
   }
   n16 = RigidBody(ComponentPose=@n15, Alpha=@n2, OverrideWorldGravity=@n1) {
    bClampLinearTranslationLimitToRefPose=true
    bOverrideWorldGravity=true
    overridePhysicsAsset=&/M_MED_SpeedDial_leftJacket_SkyDive_Physics
    simulationSpace=BaseBoneSpace
    alphaBoolBlend {blendOption=Linear}
    baseBoneRef {boneName=thigh_r}
    componentAppliedLinearAccClamp=<30000 30000 30000>
    componentLinearAccScale=<0.3 0.3 0.3>
    componentLinearVelScale=<0.2 0.2 0.5>
    simSpaceSettings {
     maxAngularAcceleration=100
     maxAngularVelocity=2
     maxLinearAcceleration=100
     maxLinearVelocity=160
     velocityScaleZ=0
     worldAlpha=0.75
     externalLinearDragV=<1 1 1>
    }
   }
   n17 = RigidBody(ComponentPose=@n16, Alpha=@n2, OverrideWorldGravity=@n1) {
    bClampLinearTranslationLimitToRefPose=true
    bOverrideWorldGravity=true
    overridePhysicsAsset=&/M_MED_SpeedDial_SkyDive_Physics
    simulationSpace=BaseBoneSpace
    alphaBoolBlend {blendOption=Linear}
    baseBoneRef {boneName=thigh_r}
    componentAppliedLinearAccClamp=<30000 30000 30000>
    componentLinearAccScale=<0.3 0.3 0.3>
    componentLinearVelScale=<0.2 0.2 0.5>
    simSpaceSettings {
     maxAngularAcceleration=100
     maxAngularVelocity=2
     maxLinearAcceleration=100
     maxLinearVelocity=160
     velocityScaleZ=0
     worldAlpha=0.75
     externalLinearDragV=<1 1 1>
    }
   }
   n18 = Control Rig!(Source=Component To Local(ComponentPose=@n17), Alpha, JacketCorrection:double=Get NewVar() -> (NewVar:double)) -> (Pose) {
    alphaBoolBlend {blendOption=Linear}
    controlRigAssetReference {blueprintRigClass=&/M_MED_DeformRig_Master_ControlRig_SpeedDial.M_MED_DeformRig_Master_ControlRig_SpeedDial_C}
   }
   n19 = "Transform (Modify) Bone - Bone: dyn_jacket_02_end_r"(
    ComponentPose=Local To Component(LocalPose=@n18)
    Alpha=1
    Translation="-0.151,-0.664,-0.078"
    Rotation="-1.675,-12.579,-6.621"
    Scale="1,1,1"
   ) {
    rotationMode=BMM_Additive
    rotationSpace=BCS_BoneSpace
    translationMode=BMM_Additive
    translationSpace=BCS_BoneSpace
    alphaBoolBlend {blendOption=Linear}
    alphaScaleBiasClamp {bInterpResult=true}
    boneToModify {boneName=dyn_jacket_02_end_r}
    rotation=<rot -1.675 -12.579 -6.621>
    translation=<-0.151 -0.664 -0.078>
   }
   n20 = "Transform (Modify) Bone - Bone: dyn_jacket_base_01_c"(ComponentPose=@n19, Alpha=1, Translation="0,-1,0", Rotation="0,0,0", Scale="1,1,1") {
    rotationMode=BMM_Additive
    rotationSpace=BCS_BoneSpace
    translationMode=BMM_Additive
    translationSpace=BCS_BoneSpace
    alphaBoolBlend {blendOption=Linear}
    alphaScaleBiasClamp {bInterpResult=true}
    boneToModify {boneName=dyn_jacket_base_01_c}
    translation {y=-1}
   }
   n21 = "Transform (Modify) Bone - Bone: dyn_jacket_base_01_c"(ComponentPose=@n20, Alpha=1, Translation="-0.000,-1.352,1.135", Rotation="0,0,0", Scale="1,1,1") {
    rotationMode=BMM_Additive
    rotationSpace=BCS_BoneSpace
    translationMode=BMM_Additive
    translationSpace=BCS_BoneSpace
    alphaBoolBlend {blendOption=Linear}
    alphaScaleBiasClamp {bInterpResult=true}
    boneToModify {boneName=dyn_jacket_base_01_c}
    translation {y=-1.352, z=1.135}
   }
   n22 = "Transform (Modify) Bone - Bone: dyn_jacket_base_02_c"(ComponentPose=@n21, Alpha=1, Translation="-0.042,-2.268,-1.238", Rotation="0.138,-0.001,-2.809", Scale="1,1,1") {
    rotationMode=BMM_Additive
    rotationSpace=BCS_BoneSpace
    translationMode=BMM_Additive
    translationSpace=BCS_BoneSpace
    alphaBoolBlend {blendOption=Linear}
    alphaScaleBiasClamp {bInterpResult=true}
    boneToModify {boneName=dyn_jacket_base_02_c}
    rotation=<rot 0.138 -0.001 -2.809>
    translation=<-0.042 -2.268 -1.238>
   }
   n23 = "Transform (Modify) Bone - Bone: dyn_jacket_base_03_r"(ComponentPose=@n22, Alpha=1, Translation="-0.029,2.681,0.468", Rotation="0.000,2.813,-0.000", Scale="1,1,1") {
    rotationMode=BMM_Additive
    rotationSpace=BCS_BoneSpace
    translationMode=BMM_Additive
    translationSpace=BCS_BoneSpace
    alphaBoolBlend {blendOption=Linear}
    alphaScaleBiasClamp {bInterpResult=true}
    boneToModify {boneName=dyn_jacket_base_03_r}
    rotation=<rot 0 2.813 0>
    translation=<-0.029 2.681 0.4679>
   }
   n24 = "Transform (Modify) Bone - Bone: deform_pec_base_l"(ComponentPose=@n23, Alpha=1, Translation="0.078,-2.057,-0.071", Rotation="-36.522,3.500,-2.085", Scale="1,1,1") {
    rotationMode=BMM_Additive
    rotationSpace=BCS_BoneSpace
    translationMode=BMM_Additive
    translationSpace=BCS_BoneSpace
    alphaBoolBlend {blendOption=Linear}
    alphaScaleBiasClamp {bInterpResult=true}
    boneToModify {boneName=deform_pec_base_l}
    rotation=<rot -36.522 3.5 -2.085>
    translation=<0.078 -2.0569 -0.071>
   }
   n25 = "Transform (Modify) Bone - Bone: deform_pec_base_l"(ComponentPose=@n24, Alpha=1, Translation="0.000,-3.069,-0.000", Rotation="0.000,5.624,-0.000", Scale="1,1,1") {
    rotationMode=BMM_Additive
    rotationSpace=BCS_BoneSpace
    translationMode=BMM_Additive
    translationSpace=BCS_BoneSpace
    alphaBoolBlend {blendOption=Linear}
    alphaScaleBiasClamp {bInterpResult=true}
    boneToModify {boneName=deform_pec_base_l}
    rotation=<rot 0 5.624 -0.0001>
    translation {y=-3.069}
   }
   n26 = "Transform (Modify) Bone - Bone: dyn_jacket_base_01_c"(ComponentPose=@n25, Alpha=1, Translation="-0.000,-1.151,0.000", Rotation="0.000,0.000,-5.000", Scale="1,1,1") {
    rotationMode=BMM_Additive
    rotationSpace=BCS_BoneSpace
    translationMode=BMM_Additive
    translationSpace=BCS_BoneSpace
    alphaBoolBlend {blendOption=Linear}
    alphaScaleBiasClamp {bInterpResult=true}
    boneToModify {boneName=dyn_jacket_base_01_c}
    rotation=<rot 0 0.0001 -5>
    translation {y=-1.151}
   }
   n27 = "Transform (Modify) Bone - Bone: dyn_jacket_base_01_l"(ComponentPose=@n26, Alpha=1, Translation="0.000,-2.378,-0.112", Rotation="-0.482,-5.509,-4.976", Scale="1,1,1") {
    rotationMode=BMM_Additive
    rotationSpace=BCS_BoneSpace
    translationMode=BMM_Additive
    translationSpace=BCS_BoneSpace
    alphaBoolBlend {blendOption=Linear}
    alphaScaleBiasClamp {bInterpResult=true}
    boneToModify {boneName=dyn_jacket_base_01_l}
    rotation=<rot -0.482 -5.509 -4.976>
    translation {y=-2.378, z=-0.112}
   }
   n28 = "Transform (Modify) Bone - Bone: dyn_jacket_base_01_r"!(ComponentPose=@n27, Alpha:float=1, Translation="-1.398,0.922,0.579", Rotation="-0.249,-2.852,-4.993", Scale="1,1,1") -> (Pose) {
    rotationMode=BMM_Additive
    rotationSpace=BCS_BoneSpace
    translationMode=BMM_Additive
    translationSpace=BCS_BoneSpace
    alphaBoolBlend {blendOption=Linear}
    alphaScaleBiasClamp {bInterpResult=true}
    boneToModify {boneName=dyn_jacket_base_01_r}
    rotation=<rot -0.249 -2.852 -4.993>
    translation=<-1.398 0.922 0.579>
   }
   n29 = "Transform (Modify) Bone - Bone: dyn_jacketMid_base_01_r"!(ComponentPose=@n28, Alpha:float=1, Translation="0.033,2.891,-0.000", Rotation="0.000,6.673,0.000", Scale="1,1,1") -> (Pose) {
    rotationMode=BMM_Additive
    rotationSpace=BCS_BoneSpace
    translationMode=BMM_Additive
    translationSpace=BCS_BoneSpace
    alphaBoolBlend {blendOption=Linear}
    alphaScaleBiasClamp {bInterpResult=true}
    boneToModify {boneName=dyn_jacketMid_base_01_r}
    rotation=<rot 0 6.673 0>
    translation=<0.033 2.891>
   }
   n30 = "Transform (Modify) Bone - Bone: dyn_jacketMid_base_01_l"(ComponentPose=@n29, Alpha=1, Translation="0.000,-1.282,-0.000", Rotation="0,0,0", Scale="1,1,1") {
    rotationMode=BMM_Additive
    rotationSpace=BCS_BoneSpace
    translationMode=BMM_Additive
    translationSpace=BCS_BoneSpace
    alphaBoolBlend {blendOption=Linear}
    alphaScaleBiasClamp {bInterpResult=true}
    boneToModify {boneName=dyn_jacketMid_base_01_l}
    translation=<0 -1.2818 0>
   }
   n31 = "Transform (Modify) Bone - Bone: dyn_jacketMid_base_01_l"(ComponentPose=@n30, Alpha=1, Translation="0,-2,0", Rotation="0,-5,0", Scale="1,1,1") {
    rotationMode=BMM_Additive
    rotationSpace=BCS_BoneSpace
    translationMode=BMM_Additive
    translationSpace=BCS_BoneSpace
    alphaBoolBlend {blendOption=Linear}
    alphaScaleBiasClamp {bInterpResult=true}
    boneToModify {boneName=dyn_jacketMid_base_01_l}
    rotation {yaw=-5}
    translation {y=-2}
   }
   n32 = "Transform (Modify) Bone - Bone: dyn_jacketMid_base_01_r"(ComponentPose=@n31, Alpha=1, Translation="0,2,0", Rotation="-0.001,-2.673,-0.000", Scale="1,1,1") {
    rotationMode=BMM_Additive
    rotationSpace=BCS_BoneSpace
    translationMode=BMM_Additive
    translationSpace=BCS_BoneSpace
    alphaBoolBlend {blendOption=Linear}
    alphaScaleBiasClamp {bInterpResult=true}
    boneToModify {boneName=dyn_jacketMid_base_01_r}
    rotation=<rot -0.001 -2.673 0>
    translation {y=2}
   }
   n33 = "Transform (Modify) Bone - Bone: dyn_jacket_base_03_l"!(
    ComponentPose=@n32
    bAlphaBoolEnabled:bool=Get bIsTargetingConsumableThrow() -> (bIsTargetingConsumableThrow:bool)
    Translation="0,-2,0"
    Rotation="0,0,0"
    Scale="1,1,1"
   ) -> (Pose) {
    alphaInputType=Bool
    rotationMode=BMM_Additive
    rotationSpace=BCS_BoneSpace
    translationMode=BMM_Additive
    translationSpace=BCS_BoneSpace
    alphaBoolBlend {blendInTime=0.2, blendOption=Linear, blendOutTime=0.2}
    alphaScaleBiasClamp {bInterpResult=true}
    boneToModify {boneName=dyn_jacket_base_03_l}
    translation {y=-2}
   }
   n34 = "Transform (Modify) Bone - Bone: dyn_jacket_base_01_c"!(ComponentPose=@n33, bAlphaBoolEnabled:bool=Get bIsCrouching(), Translation="0,-2,0", Rotation="0,0,0", Scale="1,1,1") -> (Pose) {
    alphaInputType=Bool
    rotationMode=BMM_Additive
    rotationSpace=BCS_BoneSpace
    translationMode=BMM_Additive
    translationSpace=BCS_BoneSpace
    alphaBoolBlend {blendInTime=0.2, blendOption=Linear, blendOutTime=0.2}
    alphaScaleBiasClamp {bInterpResult=true}
    boneToModify {boneName=dyn_jacket_base_01_c}
    translation {y=-2}
   }
   n35 = "Transform (Modify) Bone - Bone: deform_geometrycorrection"(ComponentPose=@n34, Alpha=1, Translation="0,0,0", Rotation="0.000,0.000,-5.000", Scale="1,0.5,0.5") {
    rotationSpace=BCS_BoneSpace
    scaleMode=BMM_Additive
    scaleSpace=BCS_BoneSpace
    translationSpace=BCS_BoneSpace
    alphaBoolBlend {blendOption=Linear}
    alphaScaleBiasClamp {bInterpResult=true}
    boneToModify {boneName=deform_geometrycorrection}
    rotation=<rot 0 0.0001 -5>
    scale {y=0.5, z=0.5}
   }
   Output Pose(Result=Component To Local(ComponentPose=@n35)) {layerGroup=None}
  ]}
 EventGraph {nodes [
   n1 = Event BlueprintUpdateAnimation()
   > n2 = Sequence!(execute) -> (then_0, then_1, then_2:Exec, then_3:Exec, then_4:Exec, then_5:Exec)
   > n3 = Gravity Override State Function DEPRECATED(
    CharacterPartAnimInstance=Self-Reference()
    GameTimeInSeconds=GetGameTimeInSeconds()
    GravityOverrideParams=Get Gravity Override Params()
    DebugJointName=Pelvis
   )
   > Set Override World Gravity(Override World Gravity:Vector=@n3) -> (Output_Get:Vector)
   n4 = Get Skeletal Mesh Component() -> (Skeletal Mesh Component:Skeletal Mesh Component Object Reference)
   n5 = Get Overlay Mat Instance Dyanmic() -> (Overlay Mat Instance Dyanmic:Material Instance Dynamic Object Reference)
   n6 = Get Material Instance Dynamic() -> (Material Instance Dynamic:Material Instance Dynamic Object Reference)
   > Set WPO Globals Function(
    Material=@n6
    Overlay Material=@n5
    Skeletal Mesh=@n4
    "Local Noise Transform  Location"="Get Local Noise Transform  Location"() -> ("Local Noise Transform  Location":Vector)
    Local Noise Offset=Get Local Noise Offset()
   )
   > Set WPO Regional Function!(
    execute
    Material:Material Instance Dynamic Object Reference=@n6
    Overlay Material:Material Instance Dynamic Object Reference=@n5
    Skeletal Mesh:Skeletal Mesh Component Object Reference=@n4
    Velocity Transform:Name=calf_r
    Velocity Interp Speed:double=Get Velocity Interp Speed()
    Collision Transform:Name=calf_r
    Collision Transform Axis:EAxis Enum
    Invert Collision Transform Axis:bool=false
    Delta Time X:double=@n1.DeltaTimeX
    Previous Velocity Transform Location:"Vector (by ref)"=Get R1 Velocity Transform Location()
    Previous Velocity Transform Velocity:"Vector (by ref)"=Get R1 Velocity Transform Velocity()
    Region:int=1
   ) -> (then)
   > Set WPO Regional Function(
    Material=@n6
    Overlay Material=@n5
    Skeletal Mesh=@n4
    Velocity Transform=calf_l
    Velocity Interp Speed=Get Velocity Interp Speed()
    Collision Transform=calf_l
    Invert Collision Transform Axis=true
    Delta Time X=@n1.DeltaTimeX
    Previous Velocity Transform Location=Get R2 Velocity Transform Location()
    Previous Velocity Transform Velocity=Get R2 Velocity Transform Velocity()
    Region=2
   )
   > SetScalarParameterValue(self=[@n6, @n5], ParameterName=ClothSim_WindStrength, Value=Get WPO Wind() -> (WPO Wind:double))
   > SetScalarParameterValue(self=[@n6, @n5], ParameterName=ClothSim_TurbulenceNoiseTileDivisor, Value=Get Turbulence Noise() -> (Turbulence Noise:double))
   n7 = Branch(
    execute=@n2.then_1
    Condition=OR Boolean!(A=Get bIsSkydiveFloating(), B=Get bIsSkydiveDiveMode(), C:bool=Get bIsParachuteOpened()) -> (ReturnValue)
   )
   > Set Velocity Interp Speed(Velocity Interp Speed:double=40) -> (Output_Get:double)
   > Set WPO Wind(WPO Wind:double=1) -> (Output_Get:double)
   > Set Turbulence Noise(Turbulence Noise:double=80) -> (Output_Get:double)
   Set Velocity Interp Speed(execute=@n7.else, Velocity Interp Speed=0)
   > Set WPO Wind(WPO Wind=0)
   > Set Turbulence Noise(Turbulence Noise=0)
   n8 = AND Boolean(A=Get bIsCrouching(), B=Get bIsMoving2D())
   Set JacketCrouchRotateCorrection(execute=@n2.then_3, JacketCrouchRotateCorrection:bool=@n8) -> (Output_Get:bool)
   n9 = GetOwnerSkeletalMeshComponent()
   > Set lowerarm(lowerarm:double=MapRangeClamped(
     Value=GetAlphaFromReferenceTransformDelta(SkeletalMeshComponent=@n9, BoneName=lowerarm_l, BaseName=upperarm_l, InputRange=(X=45,Y=63), TwistAxis="1,0,0")
     InRangeA=0
     InRangeB=1
     OutRangeB=14
    )) -> (Output_Get:double)
   n10 = GetOwnerAnimationInterface!(self:Custom Character Part Anim Instance Object Reference) -> (ReturnValue:Fort Player Animation Interface Object Reference)
   > Set is_pistol(is_pistol:bool="Equal (Enum)"(
     A:EFortWeaponCoreAnimation Enum=GetWeaponCoreAnim!(self:Fort Player Animation Interface Object Reference=@n10) -> (ReturnValue:EFortWeaponCoreAnimation Enum)
     B:EFortWeaponCoreAnimation Enum=Pistol
    )) -> (Output_Get:bool)
   n11 = Branch(execute=@n2.then_4, Condition=Get bIsInFrontEnd())
   n12 = Get PawnVelocity()
   n13 = OR Boolean!(
    A="float > float"(A=Vector Length(A=@n12))
    B=Get bIsMoving2D()
    C=Get bIsCrouching()
    D:bool=Get bIsSkyDiving()
    E:bool=Get bIsPlayingEmote()
    F:bool=Get bIsTargetingConsumableThrow()
   ) -> (ReturnValue)
   n14 = FInterpTo(
    Current=Get NewVar()
    Target="To Float (Boolean)"(InBool=NOT Boolean(A=@n13))
    DeltaTime=GetWorldDeltaSeconds!() -> (ReturnValue:double)
    InterpSpeed=20
   )
   > Set NewVar(NewVar:double=@n14) -> (Output_Get:double)
   Set NewVar(execute=@n11.else, NewVar="Clamp (Float)"(Value=float - float(A=@n14, B=0.3), Min=0, Max=1))
   Set Is Crouching(execute=@n2.then_2, Is Crouching:bool=Get bIsCrouching()) -> (Output_Get:bool)
   > Set Is Jumping(Is Jumping:bool=Get bIsJumping()) -> (Output_Get:bool)
   > n15 = Branch(Condition=Get bIsSkyDiving())
   > Set skydive Alpha(skydive Alpha:double=1) -> (Output_Get:double)
   > Set Ground Alpha(Ground Alpha:double=0) -> (Output_Get:double)
   Set skydive Alpha(execute=@n15.else, skydive Alpha=0)
   > Set Ground Alpha(Ground Alpha=1)
   n16 = Get bIsSprinting()
   Set sprinting(
    execute=@n2.then_5
    sprinting:double=FInterpTo(Current=Get sprinting() -> (sprinting:double), Target="To Float (Boolean)"(InBool=@n16), DeltaTime=GetWorldDeltaSeconds(), InterpSpeed=20)
   ) -> (Output_Get:double)
   Event OnCharacterCustomizationLoaded()
   n17 = GetPartSkeletalMeshComponent()
   > n18 = Set Skeletal Mesh Component(Skeletal Mesh Component:Skeletal Mesh Component Object Reference=@n17) -> (Output_Get:Skeletal Mesh Component Object Reference)
   > n19 = Get Dynamic Material Instance(SkeletalMeshComponent=@n18)
   > Set Material Instance Dynamic(Material Instance Dynamic:Material Instance Dynamic Object Reference=@n19) -> (Output_Get:Material Instance Dynamic Object Reference)
   > n20 = GetOrCreateDynamicOverlayMaterial(MeshComponent=@n17, MaterialName=None)
   > Set Overlay Mat Instance Dyanmic(Overlay Mat Instance Dyanmic:Material Instance Dynamic Object Reference=@n20) -> (Output_Get:Material Instance Dynamic Object Reference)
   OR Boolean!(A=@n8, B=Get bIsCrouching()) -> (ReturnValue)
  ]}
]
variables [
 Gravity Override Params {
  type=GravityOverrideParamsStruct
  default {
   skydive_FloatingDown {gravityOverride=<0 -100 100>, windAmplitude=<100 -200 100>, windFrequency=<1.5 3 3>, windOffset=<0 0 0>}
   skydive_GoingUp {gravityOverride=<0 -100 100>, windAmplitude=<100 -200 100>, windFrequency=<1.5 3 3>}
  }
 }
 Ground Alpha {type=number}
 Is Crouching {type=boolean}
 Is Jumping {type=boolean}
 is_pistol {type=boolean}
 JacketCrouchRotateCorrection {type=boolean}
 Local Noise Offset {type=Vector}
 "Local Noise Transform  Location" {type=Vector}
 lowerarm {type=number}
 Material Instance Dynamic {type=MaterialInstanceDynamic}
 NewVar {type=number}
 Overlay Mat Instance Dyanmic {type=MaterialInstanceDynamic}
 Override World Gravity {type=Vector}
 R1 Velocity Transform Location {type=Vector}
 R1 Velocity Transform Velocity {type=Vector}
 R2 Velocity Transform Location {type=Vector}
 R2 Velocity Transform Velocity {type=Vector}
 Skeletal Mesh Component {type=SkeletalMeshComponent}
 skydive Alpha {type=number}
 sprinting {type=number}
 Turbulence Noise {type=number}
 Velocity Interp Speed {type=number}
 WPO Wind {type=number}
]
